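; Quest visual for the honeypot quest interactable, playing feedback when the player's objective completes.
; After a short delay, a timeline shrinks the Lid component to near zero scale.
; A timeline event track spawns the honeypot pickup FX at the lid, then fires a delayed local burst gameplay cue on the player's pawn.
; When the timeline finishes, a helper function plays optional sound, FX and force feedback, rumbling only the first local controller despite its Splitscreen name.
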(event EventBeginPlay)

(event Collision|EventActorBeginOverlap
  (params (OtherActor :type Actor%20Object%20Reference)))

(event EventTick
  (params (DeltaSeconds :type Float%20%28single-precision%29)))

(event EventObjectiveCompleted
  (params (PlayerController :type Fort%20Player%20Controller%20Object%20Reference))
  (Utilities|FlowControl|Delay 0.3)
  (bind (_direction _shrinktime) (Timeline|Timeline_0 PlayFromStart)
    (:Update
      (Transformation|SetRelativeScale3D (Variables|Components|GetLid) (Math|Vector|Lerp(Vector) "1.000000,1.000000,1.000000" "0.010000,0.010000,0.010000" _shrinktime)))
    (:Finished
      (CallFunction|AthenaPlayQuestRewardFeedbackSplitscreen (Variables|Default|GetSound) (Variables|Default|GetFX) (Variables|Default|GetShake) self))
    (:NewTrack_0
      (bind _returnvalue (Effects|Components|ParticleSystem|SpawnFXAtLocation "/Game/Athena/Items/Consumables/HoneyPot/FX/NS_Honeypot_Pickup.NS_Honeypot_Pickup" (Transformation|GetWorldLocation (Variables|Components|GetLid))))
      (Components|Activation|Activate _returnvalue true)
      (Utilities|FlowControl|Delay 0.75)
      (GameplayCue|ExecuteGameplayCue%28Burst%29Local (Game|Player|GetPlayerPawn PlayerController) (Variables|Default|GetNewVar_0) (Variables|Default|GetNewVar_1)))))

(fn AthenaPlayQuestRewardFeedback_Splitscreen
  (params (NewParam :type Sound%20Base%20Object%20Reference) (NewParam1 :type FXSystem%20Asset%20Object%20Reference) (NewParam2 :type Force%20Feedback%20Effect%20Object%20Reference) (NewParam3 :type Actor%20Object%20Reference))
  (bind _returnvalue (Transformation|GetActorLocation NewParam3))
  (Utilities|IsValid NewParam
    (:Is%20Valid
      (Audio|SpawnSoundatLocation NewParam _returnvalue))
    (:Is%20Not%20Valid
      _))
  (Utilities|IsValid NewParam1
    (:Is%20Valid
      (Effects|Components|ParticleSystem|SpawnFXAtLocation NewParam1 _returnvalue))
    (:Is%20Not%20Valid
      _))
  (Utilities|IsValid NewParam2
    (:Is%20Valid
      (Game|Feedback|ClientPlayForceFeedback (Utilities|Array|Get(acopy) (Fortnite|GetLocalFortPlayerControllers true)) NewParam2 "None" false false false))))

(fn ConstructionScript)
